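; Weapon cosmetics Blueprint for the dual-wield Drip pickaxe (Alien Summer Male) that hooks into the wielder's jump and landing events.
; On cosmetic init, it casts the instigator to FortPlayerPawnAthena, caches it, and binds to movement-updated and landed delegates.
; The movement handler treats active jump force as jump start, and the landed handler marks jump end.
; Both handlers end in DoOnce nodes with nothing connected after them, so neither produces a visible effect.
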(event EventOnInitWeaponCosmetics
  (|CallParentFunction)
  (bind _asfort_player_pawn_athena (Utilities|Casting|CastToFortPlayerPawnAthena (Game|GetInstigator)))
  (Variables|Default|SetFort%20Player%20Pawn%20Athena _asfort_player_pawn_athena)
  (Character|BindEventtoOnCharacterMovementUpdated _asfort_player_pawn_athena (EventDispatchers|CreateEvent :!event Custom|On%20Begin%20Jump))
  (EventDispatchers|BindEventtoOnPawnLanded _asfort_player_pawn_athena (EventDispatchers|CreateEvent :!event Custom|On%20End%20Jump)))

(event Custom|On%20Begin%20Jump
  (params (DeltaSeconds :type Float%20%28single-precision%29) (OldLocation :type Vector) (OldVelocity :type Vector))
  (if (Character|IsJumpProvidingForce (Variables|Default|GetFort%20Player%20Pawn%20Athena))
    (Utilities|FlowControl|DoOnce false)))

(event Custom|On%20End%20Jump
  (params (Hit :type Hit%20Result%20Structure%20%28by%20ref%29))
  (Utilities|FlowControl|Sequence
    (:then_0
      (Utilities|FlowControl|DoOnce false))))

(fn ConstructionScript
  (|CallParentFunction))
