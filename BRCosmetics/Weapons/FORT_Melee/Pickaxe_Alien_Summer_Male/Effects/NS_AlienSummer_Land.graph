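# Niagara GPU effect for the Alien Summer male pickaxe cosmetic, named for landing, that makes ice-cream drips on a skeletal mesh.
# IceCreamDrips burst-spawns on the mesh, snaps to the depth buffer, culls by displacement or invalid sample, and runs a GBuffer color module.
# Drips001 spawns from those particles via an attribute reader, falling under gravity with point attraction.
# AlienSummer_Drips_Land and Drips emitters are disabled, leaving the IsOnGround input inactive.
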NiagaraSystem /BRCosmetics/Weapons/FORT_Melee/Pickaxe_Alien_Summer_Male/Effects/NS_AlienSummer_Land
dependencies {
 usedDataInterfaces [
  &NiagaraDataInterfaceCurve
  &NiagaraDataInterfaceArrayFloat
  &NiagaraDataInterfaceGBuffer
  &NiagaraDataInterfaceVectorCurve
  &NiagaraDataInterfaceColorCurve
  &NiagaraDataInterfaceVector2DCurve
  &NiagaraDataInterfaceParticleRead
 ]
 usedDynamicInputs []
 usedModules [&SystemState]
}
emitters {
 AlienSummer_Drips_Land {
  bEnabled=false
  simTarget=GPUComputeSim
  emitterUpdateScript [
   &EmitterState()
   &SpawnRate(
    SpawnRate=&Multiply_Float(
     A=&RandomRangeFloat(Minimum=40, Maximum=40)
     B=&FloatFromCurve(
      FloatCurve=DataInterface() {propertyValues {Curve {keys [{arriveTangent=-2.5212, interpMode=RCIM_Cubic, leaveTangent=-2.5212, tangentMode=RCTM_User, time=0, value=1}, <at 1 0>]}}}
      CurveIndex=$Emitter.NormalizedLoopAge
     )
    )
    Use Spawn Probability=true
   )
   /SetVariables(Emitter.PlayerVelocityVector:Vector3f=$Engine.Owner.Velocity) {isSetParameters=true}
   /SetVariables(Emitter.PlayerSpeed:NiagaraFloat=&VectorLength(Vector=$Emitter.PlayerVelocityVector, Fallback Vector=<0 0 0>)) {isSetParameters=true}
   &SpawnPerUnit(
    Module.Use Max Movement Threshold=true
    Module.Use Movement Tolerance=false
    Spawn Per Unit:NiagaraFloat=&Lerp_Float(
     A=40
     B=100
     Alpha=&FloatFromCurve(FloatCurve=DataInterface() {propertyValues {Curve {keys [<at 0 1>,<at 1 0>]}}}, CurveIndex=$Emitter.PlayerSpeed)
    )
    Use Spawn Probability=false
   ) {enabled=false}
   &SpawnBurst_Instantaneous(
    Spawn Count=1
    Spawn Probability=&MakeCustomFloatFromBool(Bool=$User.IsOnGround, True Float=1, False Float=0)
    Use Spawn Probability=true
   ) {enabled=false}
  ]
  particleSpawnScript [
   /SetVariables(Particles.WhichSocket:NiagaraInt32=&RandomRangeInt(Minimum=0, Maximum=3)) {isSetParameters=true}
   /Niagara/Modules/Spawn/Initialization/InitializeParticle(
    Color=<rgba 1 1 1 2>
    Lifetime=&RandomRangeFloat(Minimum=0.7, Maximum=1)
    Mesh Scale=&Multiply_VectorByFloat(
     Vector=&Multiply_VectorByFloat(
      Vector=<1 1 1>
      Float=&Multiply_Float(
       A=0.6
       B=&SelectFloatFromArray(Direct Array Index=$Particles.WhichSocket, Float Selection Array=DataInterface() {propertyValues {FloatData [1.7,1.4,1]}})
      )
     )
     Float=&RandomRangeFloat(Minimum=1, Maximum=1.3)
    )
    Module.Write Color=true
    Module.Write Lifetime=true
    Module.Write Mass=true
    Module.Write Material Random=true
    Module.Write Position=false
    Module.Write Scale=true
    Module.Write SpriteRotation=false
    Module.Write SpriteSize=true
    Position=&SimulationPosition()
    Sprite Size=&Vector2DFromFloat(Value=&RandomRangeFloat(Minimum=3, Maximum=8))
   )
   &SkeletalMeshLocation(
    Inherit Velocity=false
    Interpolate Mesh Sampling=false
    MaskSpawnGroup=false
    Mesh Sampling Type=::NewEnumerator5
    Module.Override Randomness=false
    Sampled Position Offset=<0 0 10>
    Skeletal Mesh=$User.SkeletalMesh
   )
   /SetVariables(Particles.LocalPosition:Vector3f=/Niagara/DynamicInputs/Transforms/TransformPosition(Position=Linked() {linkedVariable {name=Particles.Position, type=NiagaraPosition}}, Source Space=::World, Destination Space=::Local)) {enabled=false, isSetParameters=true}
   &GravityForce(Gravity=&Multiply_VectorByFloat(
     Vector=<1 1 -700>
     Float=&FloatFromCurve(FloatCurve=DataInterface() {propertyValues {Curve {keys [{arriveTangent=1.8039, time=0, value=1}, <at 1 1>]}}})
    )) {enabled=false}
   /SetVariables(Particles.IntPosition:NiagaraPosition=$Particles.Position) {isSetParameters=true}
   &ApplyInitialForces()
   /Niagara/Modules/Math/PlaceParticlesOnDepthBuffer_GPU(Max Z Depth=1500, Module.Use Max Z Depth=true)
   &KillParticles(Kill Particles=&SetBoolByFloatComparison(
     A=&VectorLength(Vector=&Subtract_Vector(A=$Particles.IntPosition, B=Linked() {linkedVariable {name=Particles.Position, type=NiagaraPosition}}))
     B=3
    )) {module=KillParticles001}
   &KillParticles(Kill Particles=&Boolean_NotOperation(Bool=$Output.PlaceParticlesOnDepthBuffer_GPU.ValidSample))
   &self:GBufferColor() {module=GBufferColor}
  ]
  particleUpdateScript [
   /Niagara/Modules/Update/Orientation/V2/OrientMeshToVector(Override Quat=false, UseOptionalSideAxis=true, UseOptionalUpAxis=true)
   &InheritVelocity(Module.Limit Speed=true, Velocity Scale=&Multiply_VectorByFloat(Vector=<1 1 1>, Float=-0.1)) {enabled=false, module=InheritVelocity001}
   &ParticleState()
   &InheritVelocity(
    Module.Limit Speed=true
    Velocity Limit=150
    Velocity Scale=&Multiply_VectorByFloat(
     Vector=<1 1 1>
     Float=/Niagara/DynamicInputs/Math/LerpMultipleFloats(
      Float Count:ENiagara_MultipleLerpCount=::NewEnumerator3
      1:NiagaraFloat=0.02
      2:NiagaraFloat=0
      3:NiagaraFloat=?
      4:NiagaraFloat=?
      5:NiagaraFloat=?
      Alpha:NiagaraFloat=$Particles.NormalizedAge
     )
    )
   ) {enabled=false}
   &ScaleMeshSize(Scale Factor=&Multiply_Vector(
     A=<1 1 1>
     B=&VectorFromCurve(
      VectorCurve=DataInterface() {propertyValues {
        XCurve {keys [
          <at 0 1 RCIM_Cubic>
          {arriveTangent=0.5, interpMode=RCIM_Cubic, leaveTangent=-5.3087, tangentMode=RCTM_Break, time=0.2369, value=3}
          {arriveTangent=11.709, interpMode=RCIM_Cubic, leaveTangent=11.709, tangentMode=RCTM_User, time=0.3386, value=1.5078}
          <at 1 0.5>
         ]}
        YCurve {keys [
          {arriveTangent=2.3356, interpMode=RCIM_Cubic, time=-0.0034, value=1}
          {arriveTangent=5.9046, interpMode=RCIM_Cubic, leaveTangent=-6.5667, tangentMode=RCTM_Break, time=0.2412, value=1.7614}
          {arriveTangent=-1.48, interpMode=RCIM_Cubic, leaveTangent=-1.48, tangentMode=RCTM_User, time=0.3329, value=1.3319}
          <at 1 1>
         ]}
        ZCurve {keys [
          {arriveTangent=1.8377, interpMode=RCIM_Cubic, leaveTangent=1.8377, tangentMode=RCTM_User, time=-0.0042, value=1}
          {arriveTangent=5.4767, interpMode=RCIM_Cubic, leaveTangent=-3.8259, tangentMode=RCTM_Break, time=0.244, value=1.7314}
          {arriveTangent=-1.4717, interpMode=RCIM_Cubic, leaveTangent=-1.4717, tangentMode=RCTM_User, time=0.3368, value=1.3453}
          <at 1 1>
         ]}
       }}
      Scale Curve=&Multiply_VectorByFloat(
       Vector=<1 1 1>
       Float=&Multiply_Float(
        A=0.6
        B=&SelectFloatFromArray(Direct Array Index=$Particles.WhichSocket, Float Selection Array=DataInterface() {propertyValues {FloatData [2,2,0.5]}})
       )
      )
     )
    )) {enabled=false}
   &AccelerationForce(Acceleration=<0 0 -980>) {enabled=false}
   &GravityForce() {module=GravityForce001}
   &SolveForcesAndVelocity(Clamp Velocity=false, Limit Acceleration=false)
   /SetVariables(Particles.Position=&Lerp_Vector(
     A=&TransformPosition(Position=$Particles.LocalPosition, Source Space=::Local, Destination Space=::Simulation)
     B=Linked() {linkedVariable {name=Particles.Position, type=NiagaraPosition}}
     Alpha=&FloatFromCurve(FloatCurve=DataInterface() {propertyValues {Curve {keys [<at 0.25 0>, {arriveTangent=3.8697, interpMode=RCIM_Cubic, tangentMode=RCTM_Break, time=0.5, value=1}]}}})
    )) {enabled=false, isSetParameters=true}
   &ScaleColor(
    Scale Alpha=&FloatFromCurve(FloatCurve=DataInterface() {propertyValues {Curve {keys [<at 0 1 RCIM_Cubic>,<at 1 0>]}}})
    ScaleA=true
    ScaleRGB=true
   ) {enabled=false}
   &DynamicMaterialParameters(
    Index 0 Param 1=0.8
    Index 0 Param 2=&FloatFromCurve(FloatCurve=DataInterface() {propertyValues {Curve {keys [{arriveTangent=-1.8884, interpMode=RCIM_Cubic, leaveTangent=-1.8884, tangentMode=RCTM_User, time=0, value=1}, <at 1 0.26>]}}})
    Index 0 Param 3=1
    Index 0 Param 4=1
    Module.Param0WriteEnabled=true
    Module.Param1WriteEnabled=true
    Module.Param2WriteEnabled=true
    Module.Param3WriteEnabled=true
   ) {enabled=false}
   &CameraOffset(Camera Offset Amount=10) {enabled=false}
   &ScaleColor(ScaleA=true, ScaleRGB=true) {module=ScaleColor001}
   &ScaleSpriteSize(Uniform Curve Sprite Scale=DataInterface() {propertyValues {Curve {keys [
        {interpMode=RCIM_Cubic, leaveTangentWeight=0.05, tangentMode=RCTM_User, time=0, value=0}
        {arriveTangentWeight=0.0833, interpMode=RCIM_Cubic, leaveTangentWeight=0.3333, tangentMode=RCTM_User, time=0.15, value=1}
        {arriveTangentWeight=1.3532, interpMode=RCIM_Cubic, leaveTangentWeight=0.5638, tangentMode=RCTM_User, time=0.75, value=1}
        <at 1 0 RCIM_Cubic>
       ]}}})
  ]
  rendererClasses [&NiagaraSpriteRendererProperties, &NiagaraMeshRendererProperties]
  renderers [
   {rendererClass=&NiagaraSpriteRendererProperties}
   {rendererClass=&NiagaraMeshRendererProperties, rendererIndex=1}
   {rendererClass=&NiagaraSpriteRendererProperties, rendererIndex=2}
   {rendererClass=&NiagaraMeshRendererProperties, rendererIndex=3}
   {rendererClass=&NiagaraSpriteRendererProperties, rendererIndex=4}
  ]
 }
 Drips {
  bEnabled=false
  simTarget=GPUComputeSim
  emitterUpdateScript [
   &EmitterState()
   &SpawnRate(SpawnRate=3, Use Spawn Probability=false)
   &SpawnBurst_Instantaneous(Spawn Count=2, Use Spawn Probability=false, UseLoopCountLimit=false)
   &SpawnBurst_Instantaneous(Spawn Count=2, Spawn Time=0.4474, Use Spawn Probability=false, UseLoopCountLimit=false) {module=SpawnBurst_Instantaneous001}
  ]
  particleSpawnScript [
   /Niagara/Modules/Spawn/Initialization/V2/InitializeParticle(
    Lifetime Max=3
    Lifetime Mode=::NewEnumerator1
    Sprite Size Mode=::NewEnumerator3
    Uniform Sprite Size=6
    UsePositionOffset=false
    Write Material Random=true
   )
   &CameraOffset(Camera Offset Amount=&RandomRangeFloat(Minimum=2, Maximum=2))
   &SkeletalMeshLocation(MaskSpawnGroup=false, OverrideRandomSeed=false) {enabled=false}
   &SkeletalMeshLocation(
    Inherit Velocity=false
    Interpolate Mesh Sampling=false
    MaskSpawnGroup=false
    Mesh Sampling Type=::NewEnumerator5
    Module.Override Randomness=false
    Skeletal Mesh=$User.SkeletalMesh
   ) {module=SkeletalMeshLocation001}
   &BoxLocation(
    Box Origin:NiagaraPosition=/Niagara/DynamicInputs/Helpers/SimulationPosition()
    MaskSpawnGroup=false
    Module.Override Seed=false
    Normalized Box Rotation Angle=-0.1032
    Rotate Box=true
    Rotate Box On Axis:Vector3f=<0 1 0>
    Surface Only=false
   ) {enabled=false}
   /SetVariables(Particles.IntPosition:NiagaraPosition=$Particles.Position) {isSetParameters=true, module=SetVariables_#1}
   &PlaceParticlesOnDepthBuffer_GPU(Max Z Depth=1500, Module.Use Max Z Depth=true)
   &KillParticles(Kill Particles=&SetBoolByFloatComparison(
     A=&VectorLength(Vector=&Subtract_Vector(A=$Particles.IntPosition, B=Linked() {linkedVariable {name=Particles.Position, type=NiagaraPosition}}))
     B=3
    )) {module=KillParticles001}
   &KillParticles(Kill Particles=&Boolean_NotOperation(Bool=$Output.PlaceParticlesOnDepthBuffer_GPU.ValidSample))
   &self:GBufferColor() {module=GBufferColor}
   &AddVelocity(Coordinate Space=::World, Velocity=<0 0 -1>) {enabled=false}
  ]
  particleUpdateScript [
   &ParticleState()
   &InheritVelocity(Module.Limit Speed=true, Velocity Limit=150, Velocity Scale=<0 0 0>) {enabled=false}
   &InheritSourceMovement(
    Applied Force Scale=<0 0 0>
    Applied Position Falloff Distance=4000
    ApplyForce=true
    ApplyForceFalloff=false
    ApplyPosition=true
    ApplyPositionFalloff=false
    Module.LimitForceSpeed=false
   ) {enabled=false}
   &GravityForce(Gravity=&Multiply_VectorByFloat(
     Vector=<0 0 -150>
     Float=&FloatFromCurve(FloatCurve=DataInterface() {propertyValues {Curve {keys [<at 0 0 RCIM_Cubic>, {arriveTangent=4.3272, interpMode=RCIM_Cubic, leaveTangent=4.3272, tangentMode=RCTM_User, time=0.8, value=1}]}}})
    )) {enabled=false}
   &GravityForce() {module=GravityForce001}
   &PointAttractionForce(
    Attraction Radius=50
    Attraction Strength=0.25
    Attractor Position Offset:Vector3f=<8 -5 -15>
    Falloff Exponent=0.666
    Kill Radius=5
    Kill Within Radius=false
    Module.Use Falloff=true
    UseAttractorOffset=true
   ) {enabled=false}
   &Drag(Use Linear Drag=true, Use Rotational Drag=false) {enabled=false}
   &SolveForcesAndVelocity(Clamp Velocity=false, Limit Acceleration=false)
   &ScaleColor(ScaleA=true, ScaleRGB=true)
   &ScaleSpriteSize(
    Scale Factor=&MakeVector2D(
     X=1
     Y=&FloatFromCurve(FloatCurve=DataInterface() {propertyValues {Curve {keys [<at 0 0.5>,<at 1 1>]}}})
    )
    Scale Sprite Size Mode=::NewEnumerator2
   ) {module=ScaleSpriteSize001}
   &ScaleSpriteSizeBySpeed(Max Scale Factor=<1 2>, Min Scale Factor=<0.5 1>)
   &ScaleSpriteSize(Uniform Curve Sprite Scale=DataInterface() {propertyValues {Curve {keys [
        {interpMode=RCIM_Cubic, leaveTangentWeight=0.05, tangentMode=RCTM_User, time=0, value=0}
        {arriveTangentWeight=0.0833, interpMode=RCIM_Cubic, leaveTangentWeight=0.3333, tangentMode=RCTM_User, time=0.15, value=1}
        {arriveTangentWeight=1.3532, interpMode=RCIM_Cubic, leaveTangentWeight=0.5638, tangentMode=RCTM_User, time=0.75, value=1}
        <at 1 0 RCIM_Cubic>
       ]}}})
  ]
  rendererClasses [&NiagaraSpriteRendererProperties, &NiagaraMeshRendererProperties]
  renderers [
   {rendererClass=&NiagaraSpriteRendererProperties}
   {rendererClass=&NiagaraMeshRendererProperties, rendererIndex=1}
  ]
 }
 Drips001 {
  simTarget=GPUComputeSim
  emitterUpdateScript [
   &EmitterState()
   &SpawnRate(SpawnRate=5, Use Spawn Probability=false) {enabled=false}
   &SpawnParticlesFromOtherEmitter(
    Attribute Reader=DataInterface() {propertyValues {EmitterBinding {bindingMode=Other, emitterName=IceCreamDrips}}}
    Minimum Age Spawn Threshold=0.9
    Spawn Rate=2
    Spawn Rate Per Particle Cap=1
    Use Maximum Age=false
    Use Minimum Age=false
   )
  ]
  particleSpawnScript [
   &InitializeParticle(
    Lifetime Max=0.6
    Lifetime Min=0.3
    Lifetime Mode=::NewEnumerator1
    Sprite Size Mode=::NewEnumerator3
    Uniform Sprite Size=6
    UsePositionOffset=false
    Write Material Random=true
   )
   &SampleParticlesFromOtherEmitter(Apply Sampled Alpha=false, Color Sampling=::NewEnumerator0)
   &CameraOffset(Camera Offset Amount=&RandomRangeFloat(Minimum=2, Maximum=2))
   &SkeletalMeshLocation(MaskSpawnGroup=false, OverrideRandomSeed=false) {enabled=false}
   &SkeletalMeshLocation(
    Inherit Velocity=false
    Interpolate Mesh Sampling=false
    MaskSpawnGroup=false
    Mesh Sampling Type=::NewEnumerator5
    Module.Override Randomness=false
    Sampled Position Offset=<0 0 3>
    Skeletal Mesh=$User.SkeletalMesh
   ) {module=SkeletalMeshLocation001}
   &BoxLocation(
    Box Origin:NiagaraPosition=&SimulationPosition()
    MaskSpawnGroup=false
    Module.Override Seed=false
    Normalized Box Rotation Angle=-0.1032
    Rotate Box=true
    Rotate Box On Axis:Vector3f=<0 1 0>
    Surface Only=false
   ) {enabled=false}
   /SetVariables(Particles.IntPosition:NiagaraPosition=$Particles.Position) {isSetParameters=true, module=SetVariables_#1}
   &PlaceParticlesOnDepthBuffer_GPU(Max Z Depth=1500, Module.Use Max Z Depth=true) {enabled=false}
   &KillParticles(Kill Particles=&SetBoolByFloatComparison(
     A=&VectorLength(Vector=&Subtract_Vector(A=$Particles.IntPosition, B=Linked() {linkedVariable {name=Particles.Position, type=NiagaraPosition}}))
     B=3
    )) {enabled=false, module=KillParticles001}
   &KillParticles(Kill Particles=&Boolean_NotOperation(Bool=$Output.PlaceParticlesOnDepthBuffer_GPU.ValidSample)) {enabled=false}
   &self:GBufferColor_03() {enabled=false, module=GBufferColor_03}
   &AddVelocity(Coordinate Space=::Simulation, Velocity=<0 5 -100>) {enabled=false}
  ]
  particleUpdateScript [
   &ParticleState()
   &SkeletalMeshLocation(
    Apply To Particle Position=false
    Inherit Velocity=false
    "Inherit Velocity (Scale)"=0
    Interpolate Mesh Sampling=false
    MaskSpawnGroup=false
    Mesh Sampling Type=::NewEnumerator5
    Module.Override Randomness=false
    Skeletal Mesh=$User.SkeletalMesh
   ) {module=SkeletalMeshLocation002}
   &InheritSourceMovement(
    Applied Position Falloff Distance=1000
    Applied Position Scale=&VectorFromCurve(VectorCurve=DataInterface() {propertyValues {XCurve {keys [<at 0 1>,<at 0.2 0>]}, YCurve {keys [<at 0 1>,<at 0.2 0>]}, ZCurve {keys [<at 0 0>,<at 1 0>]}}})
    ApplyForce=false
    ApplyPosition=true
    ApplyPositionFalloff=false
   )
   &InheritVelocity(
    Module.Limit Speed=true
    Velocity Limit=150
    Velocity Scale=&Multiply_VectorByFloat(Vector=<-0.2 -0.2 -0.2>, Float=0.001)
   ) {enabled=false}
   &GravityForce(Gravity=&VectorFromCurve(
     VectorCurve=DataInterface() {propertyValues {
       XCurve {keys [<at 0 0>,<at 1 0>]}
       YCurve {keys [<at 0 0>,<at 1 0>]}
       ZCurve {keys [
         {interpMode=RCIM_Cubic, tangentMode=RCTM_User, time=0, value=0}
         {arriveTangent=2.1548, interpMode=RCIM_Cubic, leaveTangent=2.1548, tangentMode=RCTM_User, time=0.5, value=1}
        ]}
      }}
     Scale Curve=<1 1 -400>
    ))
   &PointAttractionForce(
    Attraction Radius=50
    Attraction Strength=0.25
    Attractor Position Offset:Vector3f=<-8 -5 0>
    Falloff Exponent=0.666
    Kill Radius=5
    Kill Within Radius=false
    Module.Use Falloff=true
    UseAttractorOffset=true
   )
   &Drag(Use Linear Drag=true, Use Rotational Drag=false) {enabled=false}
   /SetVariables(Particles.Position=&Lerp_Vector(
     A=$Output.SkeletalMeshLocation002.SampledPosition
     B=Linked() {linkedVariable {name=Particles.Position, type=NiagaraPosition}}
     Alpha=&FloatFromCurve(FloatCurve=DataInterface() {propertyValues {Curve {keys [<at 0 0>,<at 0.3 1>]}}})
    )) {enabled=false, isSetParameters=true, module=SetVariables_#2}
   &SolveForcesAndVelocity(Clamp Velocity=false, Limit Acceleration=false)
   &ScaleColor(ScaleA=true, ScaleRGB=true)
   &ScaleSpriteSize(
    Scale Factor=&MakeVector2D(
     X=1
     Y=&FloatFromCurve(FloatCurve=DataInterface() {propertyValues {Curve {keys [<at 0 0.5>,<at 1 1>]}}})
    )
    Scale Sprite Size Mode=::NewEnumerator2
   ) {module=ScaleSpriteSize001}
   &ScaleSpriteSizeBySpeed(Max Scale Factor=<1 2>, Min Scale Factor=<0.5 1>)
   &ScaleSpriteSize(Uniform Curve Sprite Scale=DataInterface() {propertyValues {Curve {keys [
        {interpMode=RCIM_Cubic, leaveTangentWeight=0.05, tangentMode=RCTM_User, time=0, value=0}
        {arriveTangentWeight=0.0833, interpMode=RCIM_Cubic, leaveTangentWeight=0.3333, tangentMode=RCTM_User, time=0.15, value=1}
        {arriveTangentWeight=1.3532, interpMode=RCIM_Cubic, leaveTangentWeight=0.5638, tangentMode=RCTM_User, time=0.75, value=1}
        <at 1 0 RCIM_Cubic>
       ]}}})
  ]
  rendererClasses [&NiagaraSpriteRendererProperties, &NiagaraMeshRendererProperties]
  renderers [
   {rendererClass=&NiagaraSpriteRendererProperties}
   {rendererClass=&NiagaraMeshRendererProperties, rendererIndex=1}
  ]
 }
 IceCreamDrips {
  simTarget=GPUComputeSim
  emitterUpdateScript [
   &EmitterState()
   &SpawnRate(SpawnRate=2, Use Spawn Probability=false) {enabled=false}
   &SpawnBurst_Instantaneous(Use Spawn Probability=false, UseLoopCountLimit=false)
  ]
  particleSpawnScript [
   &InitializeParticle(
    Lifetime Max=4
    Lifetime Mode=::NewEnumerator1
    Sprite Size Mode=::NewEnumerator3
    Uniform Sprite Size=6
    UsePositionOffset=false
    Write Material Random=true
   )
   &CameraOffset(Camera Offset Amount=&RandomRangeFloat(Minimum=2, Maximum=2))
   &SkeletalMeshLocation(MaskSpawnGroup=false, OverrideRandomSeed=false) {enabled=false}
   &SkeletalMeshLocation(
    Inherit Velocity=false
    Interpolate Mesh Sampling=false
    MaskSpawnGroup=false
    Mesh Sampling Type=::NewEnumerator5
    Module.Override Randomness=false
    Sampled Position Offset=<0 0 5>
    Skeletal Mesh=$User.SkeletalMesh
   ) {module=SkeletalMeshLocation001}
   &BoxLocation(
    Box Origin:NiagaraPosition=&SimulationPosition()
    MaskSpawnGroup=false
    Module.Override Seed=false
    Normalized Box Rotation Angle=-0.1032
    Rotate Box=true
    Rotate Box On Axis:Vector3f=<0 1 0>
    Surface Only=false
   ) {enabled=false}
   /SetVariables(Particles.IntPosition:NiagaraPosition=$Particles.Position) {isSetParameters=true, module=SetVariables_#1}
   &PlaceParticlesOnDepthBuffer_GPU(Max Z Depth=1500, Module.Use Max Z Depth=true)
   &KillParticles(Kill Particles=&SetBoolByFloatComparison(
     A=&VectorLength(Vector=&Subtract_Vector(A=$Particles.IntPosition, B=Linked() {linkedVariable {name=Particles.Position, type=NiagaraPosition}}))
     B=3
    )) {module=KillParticles001}
   &KillParticles(Kill Particles=&Boolean_NotOperation(Bool=$Output.PlaceParticlesOnDepthBuffer_GPU.ValidSample))
   &self:GBufferColor_02() {module=GBufferColor_02}
   &AddVelocity(Coordinate Space=::World, Velocity=<0 0 -1>)
  ]
  particleUpdateScript [
   &ParticleState()
   &SkeletalMeshLocation(
    Apply To Particle Position=false
    Inherit Velocity=false
    "Inherit Velocity (Scale)"=0
    Interpolate Mesh Sampling=false
    MaskSpawnGroup=false
    Mesh Sampling Type=::NewEnumerator5
    Module.Override Randomness=false
    Skeletal Mesh=$User.SkeletalMesh
   ) {enabled=false, module=SkeletalMeshLocation002}
   &InheritSourceMovement(
    Applied Position Falloff Distance=1000
    Applied Position Scale=&VectorFromFloat(Value=&FloatFromCurve(FloatCurve=DataInterface() {propertyValues {Curve {keys [<at 0.2 1>,<at 0.3 0>]}}}))
    ApplyForce=false
    ApplyPosition=true
    ApplyPositionFalloff=false
   ) {enabled=false}
   &InheritVelocity(
    Module.Limit Speed=true
    Velocity Limit=150
    Velocity Scale=&Multiply_VectorByFloat(
     Vector=<1 1 1>
     Float=&LerpMultipleFloats(
      Float Count:ENiagara_MultipleLerpCount=::NewEnumerator3
      1:NiagaraFloat=0.002
      2:NiagaraFloat=0.5
      3:NiagaraFloat=?
      4:NiagaraFloat=?
      5:NiagaraFloat=?
      Alpha:NiagaraFloat=$Particles.NormalizedAge
     )
    )
   ) {enabled=false}
   &GravityForce(Gravity=&Multiply_VectorByFloat(
     Vector=<0 0 -150>
     Float=&FloatFromCurve(
      FloatCurve=DataInterface() {propertyValues {Curve {keys [<at 0 0 RCIM_Cubic>, {arriveTangent=4.3272, interpMode=RCIM_Cubic, leaveTangent=4.3272, tangentMode=RCTM_User, time=0.8, value=1}]}}}
      Scale Curve=0.03
     )
    ))
   &PointAttractionForce(
    Attraction Radius=50
    Attraction Strength=0.1
    Attractor Position Offset:Vector3f=<-8 -5 0>
    Falloff Exponent=0.666
    Kill Radius=5
    Kill Within Radius=false
    Module.Use Falloff=true
    UseAttractorOffset=true
   )
   &Drag(Use Linear Drag=true, Use Rotational Drag=false) {enabled=false}
   /SetVariables(Particles.Position=&Lerp_Vector(
     A=$Output.SkeletalMeshLocation002.SampledPosition
     B=Linked() {linkedVariable {name=Particles.Position, type=NiagaraPosition}}
     Alpha=&FloatFromCurve(FloatCurve=DataInterface() {propertyValues {Curve {keys [<at 0 0>,<at 0.3 1>]}}})
    )) {enabled=false, isSetParameters=true, module=SetVariables_#2}
   &SolveForcesAndVelocity(Clamp Velocity=false, Limit Acceleration=false)
   &ScaleColor(ScaleA=true, ScaleRGB=true)
   &ScaleSpriteSize(
    Scale Factor=&MakeVector2D(
     X=1
     Y=&FloatFromCurve(FloatCurve=DataInterface() {propertyValues {Curve {keys [<at 0 0.5>,<at 1 1>]}}})
    )
    Scale Sprite Size Mode=::NewEnumerator2
   ) {module=ScaleSpriteSize001}
   &ScaleSpriteSizeBySpeed(Max Scale Factor=<1 2>, Min Scale Factor=<0.5 1>)
   &ScaleSpriteSize(Uniform Curve Sprite Scale=DataInterface() {propertyValues {Curve {keys [
        {interpMode=RCIM_Cubic, leaveTangentWeight=0.05, tangentMode=RCTM_User, time=0, value=0}
        {arriveTangentWeight=0.0833, interpMode=RCIM_Cubic, leaveTangentWeight=0.3333, tangentMode=RCTM_User, time=0.15, value=1}
        {arriveTangentWeight=1.3532, interpMode=RCIM_Cubic, leaveTangentWeight=0.5638, tangentMode=RCTM_User, time=0.75, value=1}
        <at 1 0 RCIM_Cubic>
       ]}}})
  ]
  rendererClasses [&NiagaraSpriteRendererProperties, &NiagaraMeshRendererProperties]
  renderers [
   {rendererClass=&NiagaraSpriteRendererProperties}
   {rendererClass=&NiagaraMeshRendererProperties, rendererIndex=1}
  ]
 }
}
summary {
 systemName=NS_AlienSummer_Land
 emitters [
  {
   bEnabled=false
   emitterName=AlienSummer_Drips_Land
   simTarget=GPUComputeSim
   rendererClasses [&NiagaraSpriteRendererProperties, &NiagaraMeshRendererProperties]
  }
  {
   bEnabled=false
   emitterName=Drips
   simTarget=GPUComputeSim
   rendererClasses [&NiagaraSpriteRendererProperties, &NiagaraMeshRendererProperties]
  }
  {
   emitterName=IceCreamDrips
   simTarget=GPUComputeSim
   rendererClasses [&NiagaraSpriteRendererProperties, &NiagaraMeshRendererProperties]
  }
  {
   emitterName=Drips001
   simTarget=GPUComputeSim
   rendererClasses [&NiagaraSpriteRendererProperties, &NiagaraMeshRendererProperties]
  }
 ]
 userVariables [
  User.IsOnGround {type=NiagaraBool}
  User.JumpSpawn {type=NiagaraFloat, defaultValue=0, description=""}
  User.SkeletalMesh {
   type=NiagaraDataInterfaceSkeletalMesh
   defaultValue {
    struct=NiagaraExt_VariableValue_DataInterface
    value {dataInterface=&self:NiagaraDataInterfaceSkeletalMesh_0}
   }
  }
 ]
}
