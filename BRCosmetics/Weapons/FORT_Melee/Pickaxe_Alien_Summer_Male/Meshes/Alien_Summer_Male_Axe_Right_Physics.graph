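# Physics asset for the right-hand axe mesh of the Alien Summer Male pickaxe, setting up capsule bodies and joints for its drip bones.
# A three-segment drip base chain hangs off the handle bone, then forks at the last segment into a two-bone drip and a single-bone drip.
# Every joint shares the same tight swing and twist limits, which restrict the drips to small bends when simulated.
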PhysicsAsset /BRCosmetics/Weapons/FORT_Melee/Pickaxe_Alien_Summer_Male/Meshes/Alien_Summer_Male_Axe_Right_Physics
bodies {
 Melee_TwoHanded_Handle [{shapeType=Capsule, length=1, radius=1}]
 drip_01_r1_right [{shapeType=Capsule, length=0.01, radius=1, shapeName=drip_01_r1_right_capsule, center=<0.5903 0 0>}]
 drip_01_r_right [{shapeType=Capsule, length=0.3205, radius=1, shapeName=drip_01_r_right_capsule, center=<0.8187 0.0019 0.0145>}]
 drip_02_r_right [{shapeType=Capsule, length=0.01, radius=1, shapeName=drip_02_r_right_capsule, center=<0.5395 0 0>}]
 drip_base_01_right [{shapeType=Capsule, length=2.1886, radius=1.4494, shapeName=drip_base_01_right_capsule, center=<0.7403 0.1702 0.044>}]
 drip_base_02_right [{shapeType=Capsule, length=2.0444, radius=1.2648, shapeName=drip_base_02_right_capsule, center=<1.264 0.067 0.1441>}]
 drip_base_03_right [{shapeType=Capsule, length=1.2264, radius=1.0377, shapeName=drip_base_03_right_capsule, center=<0.9224 0.0204 0.2658>}]
}
constraints [
 {bone1Name=drip_base_01_right, bone2Name=Melee_TwoHanded_Handle, swing1LimitDegrees=10, swing2LimitDegrees=10, twistLimitDegrees=5}
 {bone1Name=drip_base_02_right, bone2Name=drip_base_01_right, swing1LimitDegrees=10, swing2LimitDegrees=10, twistLimitDegrees=5}
 {bone1Name=drip_base_03_right, bone2Name=drip_base_02_right, swing1LimitDegrees=10, swing2LimitDegrees=10, twistLimitDegrees=5}
 {bone1Name=drip_01_r_right, bone2Name=drip_base_03_right, swing1LimitDegrees=10, swing2LimitDegrees=10, twistLimitDegrees=5}
 {bone1Name=drip_02_r_right, bone2Name=drip_01_r_right, swing1LimitDegrees=10, swing2LimitDegrees=10, twistLimitDegrees=5}
 {bone1Name=drip_01_r1_right, bone2Name=drip_base_03_right, swing1LimitDegrees=10, swing2LimitDegrees=10, twistLimitDegrees=5}
]
