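# Physics asset for the Alien Summer Male pickaxe cosmetic, giving its drip bones simulated collision bodies and joints.
# Six small capsules cover a three-bone base chain that forks into left and right branches, with the left branch one bone longer.
# Constraints link each bone to its parent with identical tight swing and twist limits, keeping motion to a slight wobble.
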PhysicsAsset /BRCosmetics/Weapons/FORT_Melee/Pickaxe_Alien_Summer_Male/Meshes/Alien_Summer_Male_Physics
bodies {
 drip_01_l [{shapeType=Capsule, length=1, radius=1.005, shapeName=drip_01_l_capsule, center=<-1.0248 0 0.5445>}]
 drip_01_r [{shapeType=Capsule, length=1.6146, radius=0.9748, shapeName=drip_01_r_capsule, center=<-0.8591 0 -0.503>}]
 drip_02_l [{shapeType=Capsule, length=1.4689, radius=1.1094, shapeName=drip_02_l_capsule, center=<-0.8169 0 0>}]
 drip_base_01 [{shapeType=Capsule, length=1, radius=1, center=<-0.9964 0 0>}]
 drip_base_02 [{shapeType=Capsule, length=2.0986, radius=1.0152, shapeName=drip_base_02_capsule, center=<-1.1316 0 0>}]
 drip_base_03 [{shapeType=Capsule, length=1.25, radius=0.755, shapeName=drip_base_03_capsule, center=<-0.7791 0 0>}]
}
constraints [
 {bone1Name=drip_base_02, bone2Name=drip_base_01, swing1LimitDegrees=10, swing2LimitDegrees=10, twistLimitDegrees=5}
 {bone1Name=drip_base_03, bone2Name=drip_base_02, swing1LimitDegrees=10, swing2LimitDegrees=10, twistLimitDegrees=5}
 {bone1Name=drip_01_r, bone2Name=drip_base_03, swing1LimitDegrees=10, swing2LimitDegrees=10, twistLimitDegrees=5}
 {bone1Name=drip_01_l, bone2Name=drip_base_03, swing1LimitDegrees=10, swing2LimitDegrees=10, twistLimitDegrees=5}
 {bone1Name=drip_02_l, bone2Name=drip_01_l, swing1LimitDegrees=10, swing2LimitDegrees=10, twistLimitDegrees=5}
]
